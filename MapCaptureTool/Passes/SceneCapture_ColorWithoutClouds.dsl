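; Map capture pass that sets up a scene capture for color without clouds, going by its name.
; On BeginPlay it applies new show flags to the 2D capture component and destroys all low-altitude cloud actors.
; It also finds cloud strip meshes and spiral cloud Blueprints by name, destroying only the first match of each.
; The construction script adds the optional post-process material as a full-weight blendable when valid.
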(event EventBeginPlay
  (Utilities|FlowControl|Sequence
    (:then_0
      (SceneCapture|SetSceneCaptureShowFlags (Variables|DecalActor|GetCaptureComponent2D) (Variables|Default|GetNew%20Show%20Flags))
      (bind _outactors (Actor|GetAllActorsOfClass "/Game/Environments/Helios/TechArt/LowAltitudeClouds/V2/BP_LowAltitudeClouds_V2.BP_LowAltitudeClouds_V2_C"))
      (for _array_element _outactors
        (Actor|DestroyActor _array_element))
      (bind _returnvalue (Fortnite|Utility|FindActorsContainingName "/Script/FortniteGame.FortStaticMeshActor" "SM_CloudStrip"))
      (if (Utilities|Array|IsNotEmpty _returnvalue)
        (Actor|DestroyActor (Utilities|Array|Get(acopy) _returnvalue))))
    (:then_1
      (bind _returnvalue_2 (Fortnite|Utility|FindActorsContainingName "/Hermes/Techart/SpiralClouds/BP_SpiralClouds.BP_SpiralClouds_C" "BP_SpiralClouds"))
      (if (Utilities|Array|IsNotEmpty _returnvalue_2)
        (Actor|DestroyActor (Utilities|Array|Get(acopy) _returnvalue_2))))))

(fn ConstructionScript
  (Utilities|IsValid (Variables|Default|GetPostProcessMaterial)
    (:Is%20Valid
      (Rendering|AddorUpdateBlendable (Variables|Default|GetPostProcess) (Variables|Default|GetPostProcessMaterial) 1.0 :!class /Script/Engine.PostProcessComponent))))
