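; Map capture pass Blueprint that, per its name, captures scene color with water hidden.
; Before the capture event it runs the console command setting r.Water.Enabled to 0, and after capture it sets it back to 1.
; The construction script adds the configured post process material to its post process component at full weight when that material is valid.
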(event EventBeginPlay)

(event EventAfterCaptureEvent
  (Development|ExecuteConsoleCommand "r.Water.Enabled 1"))

(event EventBeforeCaptureEvent
  (Development|ExecuteConsoleCommand "r.Water.Enabled 0"))

(fn ConstructionScript
  (Utilities|IsValid (Variables|Default|GetPostProcessMaterial)
    (:Is%20Valid
      (Rendering|AddorUpdateBlendable (Variables|Default|GetPostProcess) (Variables|Default|GetPostProcessMaterial) 1.0 :!class /Script/Engine.PostProcessComponent))))
