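; Scene capture pass in the map capture tool that renders Fortnite POI volumes for the minimap.
; Before capture, it spawns a cube-volume actor at each FortPoiVolume's brush transform and copies that brush's mesh into the actor's dynamic mesh via Geometry Script.
; After capture, it destroys the spawned proxies.
; The construction script adds the post-process material, if valid, as a blendable on its post-process component.
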(event EventBeforeCaptureEvent
  (bind _outactors (Actor|GetAllActorsOfClass "/Script/FortniteGame.FortPoiVolume"))
  (for _array_element _outactors
    (bind _returnvalue (Game|SpawnActorfromClass "/MapCaptureTool/Capture_Minimap_CubeVolume.Capture_Minimap_CubeVolume_C" (Transformation|GetWorldTransform (Variables|/Script/Engine.Brush|GetBrushComponent _array_element)) "AlwaysSpawn" "MultiplyWithRoot" self 0))
    (Utilities|Array|Add (Variables|Default|GetSpawnedActors) _returnvalue)
    (bind _returnvalue_2 (DynamicMeshComponent|GetDynamicMesh (Variables|/MapCaptureTool/Capture_Minimap_CubeVolume.Capture_Minimap_CubeVolume_C|GetDynamicMesh _returnvalue)))
    (GeometryScript|Scene|CopyMeshfromComponent (Variables|/Script/Engine.Brush|GetBrushComponent _array_element) _returnvalue_2)))

(event EventAfterCaptureEvent
  (for _array_element (Variables|Default|GetSpawnedActors)
    (Actor|DestroyActor _array_element)))

(fn ConstructionScript
  (Utilities|IsValid (Variables|Default|GetPostProcessMaterial)
    (:Is%20Valid
      (Rendering|AddorUpdateBlendable (Variables|Default|GetPostProcess) (Variables|Default|GetPostProcessMaterial) 1.0 :!class /Script/Engine.PostProcessComponent))))
