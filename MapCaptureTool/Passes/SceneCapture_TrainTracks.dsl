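; Map capture pass Blueprint that isolates train tracks for a 2D scene capture in the map capture tool.
; On BeginPlay it scans landscape spline mesh actors and enables custom depth with stencil value 1 on components whose mesh is in StaticMeshesToFilter, then applies new show flags to the capture component.
; The construction script adds a valid post-process material to the PostProcess component, presumably to read that stencil mask.
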(event EventBeginPlay
  (bind _outactors (Actor|GetAllActorsOfClass "/Script/Landscape.LandscapeSplineMeshesActor"))
  (for _array_element _outactors
    (for _array_element_2 (Variables|/Script/Landscape.LandscapeSplineMeshesActor|GetStaticMeshComponents _array_element)
      (if (Utilities|Array|ContainsItem (Variables|Default|GetStaticMeshesToFilter) (Variables|/Script/Engine.StaticMeshComponent|GetStaticMesh _array_element_2))
        (Rendering|SetRenderCustomDepth _array_element_2 true)
        (Rendering|SetCustomDepthStencilValue _array_element_2 1)
        (Rendering|SetCustomDepthStencilWriteMask _array_element_2 "ERSM_255"))))
  (SceneCapture|SetSceneCaptureShowFlags (Variables|DecalActor|GetCaptureComponent2D) (Variables|Default|GetNew%20Show%20Flags)))

(fn ConstructionScript
  (Utilities|IsValid (Variables|Default|GetPostProcessMaterial)
    (:Is%20Valid
      (Rendering|AddorUpdateBlendable (Variables|Default|GetPostProcess) (Variables|Default|GetPostProcessMaterial) 1.0 :!class /Script/Engine.PostProcessComponent))))
