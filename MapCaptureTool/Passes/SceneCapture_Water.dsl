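; Scene capture Blueprint for the map capture tool's water pass, preparing level water bodies for capture.
; On BeginPlay it applies custom show flags to the capture component and creates one dynamic instance of a configured material.
; It assigns that instance to every FortWaterBodyBP's water component and enables custom depth rendering on each.
; The construction script adds a post-process material, when valid, as a full-weight blendable.
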(event EventBeginPlay
  (SceneCapture|SetSceneCaptureShowFlags (Variables|DecalActor|GetCaptureComponent2D) (Variables|Default|GetNew%20Show%20Flags))
  (bind _returnvalue (Rendering|Material|CreateDynamicMaterialInstance (Variables|Default|GetMaterialToAssignToWater) "None" "None" :!class /Script/Engine.KismetMaterialLibrary))
  (bind _outactors (Actor|GetAllActorsOfClass "/Game/Athena/Apollo/Environments/Landscape/Water/FortWaterBodyBP.FortWaterBodyBP_C"))
  (for _array_element _outactors
    (Rendering|SetWaterMaterial (Variables|/Script/Water.WaterBody|GetWaterBodyComponent _array_element) _returnvalue)
    (Rendering|SetRenderCustomDepth (Variables|/Script/Water.WaterBody|GetWaterBodyComponent _array_element) true)))

(fn ConstructionScript
  (Utilities|IsValid (Variables|Default|GetPostProcessMaterial)
    (:Is%20Valid
      (Rendering|AddorUpdateBlendable (Variables|Default|GetPostProcess) (Variables|Default|GetPostProcessMaterial) 1.0 :!class /Script/Engine.PostProcessComponent))))
